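# Physics asset for the Love Llama backpack cosmetic, defining collision bodies and joints for its two dangling ribbons.
# A box body on backpack_root anchors left and right chains of six capsule bodies each, one per ribbon bone.
# Constraints link each segment to the one before it, with tight swing and twist limits along the chain.
# Only the joints at the root allow a wider swing on one axis, so each ribbon pivots mainly from its base.
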PhysicsAsset /BRCosmetics/Accessories/FORT_Backpacks/M_MED_Love_Llama_Backpack/Meshes/M_MED_Love_Llama_Backpack_Physics
bodies {
 backpack_root [{shapeType=Box, extentX=33.8095, extentY=7.5, extentZ=27.5747, center=<-7.8852 -3.9223 -0.0004>, rotation=<rot 0 5.0008 0>}]
 dyn_l_ribbon_01 [{shapeType=Capsule, length=1.25, radius=1.75, center=<5 0 0>, rotation=<rot 90 0 0>}]
 dyn_l_ribbon_02 [{shapeType=Capsule, length=3.75, radius=1.25, center=<3.4433 -0.5 0>, rotation=<rot 90 0 0>}]
 dyn_l_ribbon_03 [{shapeType=Capsule, length=4, radius=2.5, center=<3 0 3>, rotation=<rot 90 0 0>}]
 dyn_l_ribbon_04 [{shapeType=Capsule, length=4, radius=3, center=<3 0 3>, rotation=<rot 90 0 0>}]
 dyn_l_ribbon_05 [{shapeType=Capsule, length=5.5, radius=2.25, center=<3 0 3>, rotation=<rot 90 0 0>}]
 dyn_l_ribbon_06 [{shapeType=Capsule, length=3.25, radius=2.5, center=<3 0 3>, rotation=<rot 90 0 0>}]
 dyn_r_ribbon_01 [{shapeType=Capsule, length=1.5, radius=1.75, center=<5 0 1>, rotation=<rot 90 0 0>}]
 dyn_r_ribbon_02 [{shapeType=Capsule, length=3.75, radius=1.25, center=<3 0.5 0>, rotation=<rot 90 0 0>}]
 dyn_r_ribbon_03 [{shapeType=Capsule, length=4, radius=2.25, center=<3 0.3 3>, rotation=<rot 90 0 0>}]
 dyn_r_ribbon_04 [{shapeType=Capsule, length=4, radius=2.75, center=<3 0.15 3>, rotation=<rot 90 0 0>}]
 dyn_r_ribbon_05 [{shapeType=Capsule, length=5, radius=2, center=<3 0 3>, rotation=<rot 90 0 0>}]
 dyn_r_ribbon_06 [{shapeType=Capsule, length=4.75, radius=2.25, center=<3 0 3>, rotation=<rot 90 0 0>}]
}
constraints [
 {bone1Name=dyn_r_ribbon_02, bone2Name=dyn_r_ribbon_01, swing1LimitDegrees=2, swing2LimitDegrees=5, twistLimitDegrees=2}
 {bone1Name=dyn_r_ribbon_03, bone2Name=dyn_r_ribbon_02, swing1LimitDegrees=2, swing2LimitDegrees=5, twistLimitDegrees=2}
 {bone1Name=dyn_r_ribbon_04, bone2Name=dyn_r_ribbon_03, swing1LimitDegrees=2, swing2LimitDegrees=5, twistLimitDegrees=2}
 {bone1Name=dyn_r_ribbon_05, bone2Name=dyn_r_ribbon_04, swing1LimitDegrees=2, swing2LimitDegrees=5, twistLimitDegrees=2}
 {bone1Name=dyn_r_ribbon_06, bone2Name=dyn_r_ribbon_05, swing1LimitDegrees=2, swing2LimitDegrees=5, twistLimitDegrees=2}
 {bone1Name=dyn_l_ribbon_01, bone2Name=backpack_root, swing1LimitDegrees=2, swing2LimitDegrees=30, twistLimitDegrees=2}
 {bone1Name=dyn_l_ribbon_02, bone2Name=dyn_l_ribbon_01, swing1LimitDegrees=2, swing2LimitDegrees=5, twistLimitDegrees=2}
 {bone1Name=dyn_l_ribbon_03, bone2Name=dyn_l_ribbon_02, swing1LimitDegrees=2, swing2LimitDegrees=5, twistLimitDegrees=2}
 {bone1Name=dyn_l_ribbon_04, bone2Name=dyn_l_ribbon_03, swing1LimitDegrees=2, swing2LimitDegrees=5, twistLimitDegrees=2}
 {bone1Name=dyn_l_ribbon_05, bone2Name=dyn_l_ribbon_04, swing1LimitDegrees=2, swing2LimitDegrees=5, twistLimitDegrees=2}
 {bone1Name=dyn_l_ribbon_06, bone2Name=dyn_l_ribbon_05, swing1LimitDegrees=2, swing2LimitDegrees=5, twistLimitDegrees=2}
 {bone1Name=dyn_r_ribbon_01, bone2Name=backpack_root, swing1LimitDegrees=2, swing2LimitDegrees=30, twistLimitDegrees=2}
]
