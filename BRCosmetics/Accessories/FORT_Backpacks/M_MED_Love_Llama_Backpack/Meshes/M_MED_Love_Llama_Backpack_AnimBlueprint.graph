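# Animation Blueprint for the Love Llama back bling, adding secondary motion to its bows and ribbons.
# When rigid body is enabled, left-bow AnimDynamics is mirrored to the right, spine bend rotates the root, then RigidBody simulates the physics asset.
# Otherwise, posed ribbons use Trail controllers with planar limits.
# The update event computes spine-driven rotation and state-based gravity with skydive settings, forcing straight-down gravity when DBNO.
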AnimBlueprint /BRCosmetics/Accessories/FORT_Backpacks/M_MED_Love_Llama_Backpack/Meshes/M_MED_Love_Llama_Backpack_AnimBlueprint
target_skeleton=&/M_MED_Love_Llama_Backpack_Skeleton
graphs [
 AnimGraph {nodes [
   n1 = Anim Dynamics - Bone: dyn_l_bow_01(ComponentPose=Local To Component(LocalPose=Use cached pose 'In Pose'()), Alpha=1, ExternalForce="0,0,0") {
    angularDampingOverride=0.99
    bGravityOverrideInSimSpace=true
    bOverrideAngularDamping=true
    bUseGravityOverride=true
    alphaBoolBlend {blendOption=Linear}
    boundBone {boneName=dyn_l_bow_01}
    componentLinearAccScale=<0.5 0.5 0.5>
    componentLinearVelScale=<0.5 0.5 0.5>
    gravityOverride {z=-980}
    physicsBodyDefinitions [{
      sphereCollisionRadius=0
      boundBone {boneName=dyn_l_bow_01}
      boxExtents=<5 3 5>
      constraintSetup {angularLimitsMax=<0 0 10>, angularLimitsMin=<0 0 -15>, angularTarget=<0 0 0>}
      localJointOffset=<0 0.5 0>
     }]
   }
   n2 = Copy Bone Delta - Source Bone: dyn_l_bow_01 - Target Bone: dyn_r_bow_01!(
    ComponentPose:Component Space Pose Link Structure=@n1
    Alpha:float=1
    bCopyTranslation:bool=False
    bCopyRotation:bool=true
    bCopyScale:bool=False
   ) -> (Pose:Component Space Pose Link Structure) {
    is=title AnimGraphNode_CopyBoneDelta
    rotationMultiplier=-1
    alphaBoolBlend {blendOption=Linear}
    sourceBone {boneName=dyn_l_bow_01}
    targetBone {boneName=dyn_r_bow_01}
   }
   n3 = "Transform (Modify) Bone - Bone: backpack_root"(ComponentPose=@n2, Alpha=1, Translation="0,0,0", Rotation=Get backpack_rotate(), Scale="1,1,1") {rotationMode=BMM_Additive, rotationSpace=BCS_BoneSpace, alphaBoolBlend {blendOption=Linear}, boneToModify {boneName=backpack_root}}
   n4 = RigidBody!(ComponentPose=@n3, Alpha, OverrideWorldGravity:Vector=Get GravityOverride(), ExternalForce) -> (Pose) {
    bClampLinearTranslationLimitToRefPose=true
    bOverrideWorldGravity=true
    overridePhysicsAsset=&/M_MED_Love_Llama_Backpack_Physics
    alphaBoolBlend {blendOption=Linear}
    componentAppliedLinearAccClamp=<800 800 800>
    componentLinearAccScale=<0.25 0.35 0.35>
    componentLinearVelScale=<0.3 0.3 0.3>
    overrideWorldGravity {z=-100}
   }
   n5 = "Transform (Modify) Bone - Bone: dyn_r_ribbon_01"(
    ComponentPose=Local To Component(LocalPose=Use cached pose 'In Pose'())
    Alpha=1
    Translation="0,0,0"
    Rotation="-38,0.000,0.000"
    Scale="1,1,1"
   ) {
    rotationMode=BMM_Additive
    rotationSpace=BCS_BoneSpace
    alphaBoolBlend {blendOption=Linear}
    boneToModify {boneName=dyn_r_ribbon_01}
    rotation {pitch=-38}
   }
   n6 = "Transform (Modify) Bone - Bone: dyn_l_ribbon_01"(ComponentPose=@n5, Alpha=1, Translation="0,0,0", Rotation="-30,-6,0", Scale="1,1,1") {
    rotationMode=BMM_Additive
    rotationSpace=BCS_BoneSpace
    alphaBoolBlend {blendOption=Linear}
    boneToModify {boneName=dyn_l_ribbon_01}
    rotation {pitch=-30, yaw=-6}
   }
   n7 = "Transform (Modify) Bone - Bone: dyn_r_ribbon_02"(ComponentPose=@n6, Alpha=1, Translation="0,0,0", Rotation="-10.000,0.000,0.000", Scale="1,1,1") {
    rotationMode=BMM_Additive
    rotationSpace=BCS_BoneSpace
    alphaBoolBlend {blendOption=Linear}
    boneToModify {boneName=dyn_r_ribbon_02}
    rotation {pitch=-10}
   }
   n8 = "Transform (Modify) Bone - Bone: dyn_l_ribbon_02"(ComponentPose=@n7, Alpha=1, Translation="0,0,0", Rotation="-15,0.000,0.000", Scale="1,1,1") {
    rotationMode=BMM_Additive
    rotationSpace=BCS_BoneSpace
    alphaBoolBlend {blendOption=Linear}
    boneToModify {boneName=dyn_l_ribbon_02}
    rotation {pitch=-15}
   }
   n9 = Trail controller - Bone: dyn_r_ribbon_06(ComponentPose=@n8, Alpha=1) {
    bUsePlanarLimit=true
    chainLength=6
    relaxationSpeedScale=1.5
    planarLimits [{drivingBone {boneName=backpack_root}, planeTransform {location=<0 -7 0>, rotation=<rot 0 5 -90>, scale=<1 1 1>}}]
    rotationLimits [
     {limitMax=<180 180 180>}
     {limitMax=<180 180 180>}
     {limitMax=<180 180 180>}
     {limitMax=<180 180 180>}
     {limitMax=<180 180 180>}
     {limitMax=<180 180 180>}
    ]
    rotationOffsets [<0 0 0>, <0 0 0>, <0 0 0>, <0 0 0>, <0 0 0>, <0 0 0>]
    trailBone {boneName=dyn_r_ribbon_06}
   }
   n10 = Trail controller - Bone: dyn_l_ribbon_06(ComponentPose=@n9, Alpha=1) {
    bUsePlanarLimit=true
    chainLength=6
    planarLimits [{drivingBone {boneName=backpack_root}, planeTransform {location=<0 -8 0>, rotation=<rot 0 5.0001 -90>, scale=<1 1 1>}}]
    rotationLimits [
     {limitMax=<180 180 180>}
     {limitMax=<180 180 180>}
     {limitMax=<180 180 180>}
     {limitMax=<180 180 180>}
     {limitMax=<180 180 180>}
     {limitMax=<180 180 180>}
    ]
    rotationOffsets [<0 0 0>, <0 0 0>, <0 0 0>, <0 0 0>, <0 0 0>, <0 0 0>]
    trailBone {boneName=dyn_l_ribbon_06}
   }
   Output Pose(Result=Blend Poses by bool(BlendPose_0=Component To Local(ComponentPose=@n4), BlendPose_1=Component To Local(ComponentPose=@n10), bActiveValue=Get bRigidBodyEnabled()) {blendType=Linear}) {layerGroup=None}
   Save cached pose 'In Pose'(Pose=Local Space Ref Pose()) {cacheName=In Pose}
  ]}
 EventGraph {nodes [
   Event BlueprintUpdateAnimation()
   n1 = GetOwnerSkeletalMeshComponent()
   n2 = BreakRotator(InRot=BreakTransform(InTransform=GetDeltaTransformFromRefPose(self=@n1, BoneName=spine_05, BaseName=spine_01)).Rotation)
   > Set backpack_rotate(backpack_rotate=MakeRotator(Yaw=MapRangeClamped(Value=@n2.Yaw, InRangeA=0, InRangeB=-30, OutRangeB=30)))
   > n3 = Gravity Override State Function DEPRECATED(CharacterPartAnimInstance=Self-Reference(), GameTimeInSeconds=GetGameTimeInSeconds(), GravityOverrideParams=Get Gravity Override Params())
   > Set GravityOverride(GravityOverride=@n3)
   > n4 = Branch(Condition=Get bIsDBNO())
   > Set GravityOverride(GravityOverride="0,0,-980")
   Set GravityOverride(execute=@n4.else, GravityOverride=@n3)
  ]}
]
variables [
 backpack_rotate {type=Rotator}
 Gravity Override Params {
  type=GravityOverrideParamsStruct
  default {
   skydive_FloatingDown {gravityOverride=<0 -700 -100>, windAmplitude=<200 -500 100>, windOffset=<0 -600 100>}
   skydive_GoingUp {gravityOverride=<0 -300 -100>, windAmplitude=<200 -500 200>}
   skydive_Parachuting {gravityOverride=<0 -500 -980>, windAmplitude=<200 -300 400>}
  }
 }
 GravityOverride {type=Vector}
]
